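; Mannequin manager that dresses sequence mannequins in the character and backpack cosmetics of a vehicle's occupants.
; On BeginPlay, after one tick, it finds GreaseClip sequence actors binding it and fills team slots 0 and 1 from their vehicle's driver and passengers.
; SetupMannequins copies each pawn's loadout and variants onto two child-actor mannequins, but only toggles visibility on the first.
; The construction script runs this setup in editor using a debug count.
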(event EventBeginPlay
  (bind _teammembers (Variables|Transient|GetTeamMembers))
  ;[0 0.15 0.03648 0.5] BeginPlay: Populating the mannequin setup.
  (|CallParentFunction)
  (Utilities|FlowControl|DelayUntilNextTick)
  (bind _outactors (Actor|GetAllActorsOfClass "/GreaseClipDunmurry/ProtoSeq/GreaseClipLevelSequenceActor.GreaseClipLevelSequenceActor_C"))
  (for _array_element _outactors
    (bind _sequenceplayer (Variables|/Script/LevelSequence.LevelSequenceActor|GetSequencePlayer _array_element)
      (:then
        (bind _returnvalue (Sequencer|Player|GetObjectBindings _sequenceplayer self :!class /Script/MovieScene.MovieSceneSequencePlayer))
        (if (not (Utilities|Array|IsEmpty _returnvalue))
          (bind _returnvalue_2 (Vehicle|GetDriver(Message) (Variables|/GreaseClipDunmurry/ProtoSeq/GreaseClipLevelSequenceActor.GreaseClipLevelSequenceActor_C|GetAssociated%20Vehicle _array_element)))
          (bind _returnvalue_3 (Fortnite|Utilities|AttemptGetPlayerStateforPawn _returnvalue_2))
          (bind _asfort_player_state_athena (Utilities|Casting|CastToFortPlayerStateAthena _returnvalue_3))
          (Utilities|Array|SetArrayElem (Variables|Transient|GetTeamMembers) 0 _asfort_player_state_athena true)
          (bind _returnvalue_4 (Vehicle|GetAllNonDriverPassengers(Message) (Variables|/GreaseClipDunmurry/ProtoSeq/GreaseClipLevelSequenceActor.GreaseClipLevelSequenceActor_C|GetAssociated%20Vehicle _array_element)))
          (for _array_element_2 _returnvalue_4
            (bind _returnvalue_5 (Fortnite|Utilities|AttemptGetPlayerStateforPawn _array_element_2))
            (bind _asfort_player_state_athena_2 (Utilities|Casting|CastToFortPlayerStateAthena _returnvalue_5))
            (Utilities|Array|SetArrayElem (Variables|Transient|GetTeamMembers) 1 _asfort_player_state_athena_2 true))))
      (:else)))
  (Utilities|Array|RemoveItem _teammembers) ; Remove empty spaces if any.
  (CallFunction|SetupMannequins (Utilities|Array|Length _teammembers) _teammembers))

(fn SetupMannequins
  (params (Count :type Integer) (TeamMembers :type Array%20of%20Fort%20Player%20State%20Athena%20Object%20References))
  (bind _mannequinchildactor_02 (Variables|Default|GetMannequinChildActor_02))
  (bind _mannequinchildactor_01 (Variables|Default|GetMannequinChildActor_01))
  (bind _array (Utilities|Array|MakeArray _mannequinchildactor_01 _mannequinchildactor_02 0 0 0))
  (Utilities|FlowControl|Sequence
    (:then_0
      (Rendering|SetVisibility (Variables|Default|GetMannequinChildActor_01) false false)
      (switch int (Math|Integer|Clamp(Integer) Count 1 2)
        (:1
          (Rendering|SetVisibility (Variables|Default|GetMannequinChildActor_01) true false))
        (:2
          (Rendering|SetVisibility (Variables|Default|GetMannequinChildActor_01) true false))))
    (:then_1
      (for (_array_element _array_index) TeamMembers
        (Utilities|IsValid _array_element
          (:Is%20Valid
            (if (Utilities|Array|IsValidIndex _array _array_index)
              (Utilities|IsValid (PlayerState|GetCurrentPawn _array_element)
                (:Is%20Valid
                  (bind _childactor (Variables|/Script/Engine.ChildActorComponent|GetChildActor (Utilities|Array|Get(acopy) _array _array_index))
                    (:then
                      (bind _asgrease_clip_mannequin (Utilities|Casting|CastToGreaseClip_Mannequin _childactor))
                      (Customization|SetAthenaBackpackandCharacterDefandVariants _asgrease_clip_mannequin (CosmeticLoadout|GetItemDefinitionTypefromLoadoutwithActor (PlayerState|GetCurrentPawn _array_element) "(TagName=\"LoadoutSlot.Backpack\")" "/Script/FortniteGame.AthenaBackpackItemDefinition") (CosmeticLoadout|GetActiveVariantsForItemFromActor (CosmeticLoadout|GetItemDefinitionTypefromLoadoutwithActor (PlayerState|GetCurrentPawn _array_element) "(TagName=\"LoadoutSlot.Backpack\")" "/Script/FortniteGame.AthenaBackpackItemDefinition") (PlayerState|GetCurrentPawn _array_element)) (CosmeticLoadout|GetItemDefinitionTypefromLoadoutwithActor (PlayerState|GetCurrentPawn _array_element) "(TagName=\"LoadoutSlot.Character\")" "/Script/FortniteGame.AthenaCharacterItemDefinition") (CosmeticLoadout|GetActiveVariantsForItemFromActor (CosmeticLoadout|GetItemDefinitionTypefromLoadoutwithActor (PlayerState|GetCurrentPawn _array_element) "(TagName=\"LoadoutSlot.Character\")" "/Script/FortniteGame.AthenaCharacterItemDefinition") (PlayerState|GetCurrentPawn _array_element))))
                    (:else)))))))))))

(fn ConstructionScript
  (if (Editor|InEditor)
    (CallFunction|SetupMannequins (Variables|Default|GetDebugMannequinCount))))
