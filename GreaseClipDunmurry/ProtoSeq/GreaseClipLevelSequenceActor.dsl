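; A level sequence actor, GreaseClipLevelSequenceActor, that plays its sequence relative to its own transform and stops it when the owning player's pawn changes.
; On BeginPlay it sets the default instance data's transform origin to the actor transform, then binds to the owning Fort player controller's pawn-changed event.
; That event force-stops the sequence player, covering cases such as a reboot.
; EndPlay unbinds the event.
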(event EventBeginPlay
  (Utilities|FlowControl|Sequence
    (:then_0
      (bind _asdefault_level_sequence_instance_data (Utilities|Casting|CastToDefaultLevelSequenceInstanceData (Variables|General|GetDefaultInstanceData)))
      (Variables|/Script/LevelSequence.DefaultLevelSequenceInstanceData|SetTransformOrigin (Transformation|GetActorTransform) _asdefault_level_sequence_instance_data))
    (:then_1
      ; Force stop if owned pawn changes (IE rebooted)
      (bind _asfort_player_controller_athena (Utilities|Casting|CastToFortPlayerControllerAthena (Actor|GetOwner)))
      (EventDispatchers|AssignOnFortPawnChangedEvent _asfort_player_controller_athena (delegate Custom|OnFortPawnChangedEvent_Event)))))

(event Custom|OnFortPawnChangedEvent_Event
  (params (NewPawn :type Fort%20Pawn%20Object%20Reference))
  ; Force stop if owned pawn changes (IE rebooted)
  (bind _sequenceplayer (Variables|Playback|GetSequencePlayer)
    (:then
      (Sequencer|Player|Stop _sequenceplayer :!class /Script/MovieScene.MovieSceneSequencePlayer))
    (:else)))

(event EventEndPlay
  (params (EndPlayReason :type EEndPlayReason%20Enum :default "Destroyed"))
  ; Force stop if owned pawn changes (IE rebooted)
  (bind _asfort_player_controller_athena (Utilities|Casting|CastToFortPlayerControllerAthena (Actor|GetOwner)))
  (EventDispatchers|UnbindEventfromOnFortPawnChangedEvent _asfort_player_controller_athena (delegate Custom|OnFortPawnChangedEvent_Event)))

(fn ConstructionScript)
