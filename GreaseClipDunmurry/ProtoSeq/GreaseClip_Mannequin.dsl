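; GreaseClip_Mannequin is a skeletal mesh actor Blueprint for a mannequin that uses a driver idle pose from a sports vehicle animation set.
; On BeginPlay it calls SetAnimation on its skeletal mesh component with a hard-referenced Valet NewSports idle driver animation, with no explicit play call.
; The overlap and tick events and the construction script are empty.
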(event EventBeginPlay
  (Components|Animation|SetAnimation (Variables|SkeletalMeshActor|GetSkeletalMeshComponent) "/Valet/Animation/Player/SportsNew/ValetVehicle_NewSports_Idle_Driver_CMM.ValetVehicle_NewSports_Idle_Driver_CMM"))

(event Collision|EventActorBeginOverlap
  (params (OtherActor :type Actor%20Object%20Reference)))

(event EventTick
  (params (DeltaSeconds :type Float%20%28single-precision%29)))

(fn ConstructionScript)
