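# Unlit GreaseClip time tunnel material: a glowing, scrolling tunnel with caustics, energy bands and sparks.
# A panning normal map gets fake directional lighting, tinted and combined with Fresnel-mapped caustics and a distortion texture.
# A V gradient with animated bands and noise masks that color and a curve-atlas tint, then two spark layers and a background gradient are added.
# The emissive sum feeds MF_FX_EmissiveAdjustments, while a DynamicParameter FadeIn blend to FadeColor branches separately.
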Material /GreaseClipDunmurry/ProtoSeq/Materials/M_GreaseClip_TimeTunnel
graph {
 properties {bAutomaticallySetUsageInEditor=false, shadingModel=MSM_Unlit, usedShadingModels=Unlit}
 nodes [
  Constant() {r=0.1}
  n1 = VectorParameter() {parameterName=DimColor, defaultValue=<rgba 0.0437 0.4179 1 1>}
  n2 = VectorParameter() {parameterName=LightColor, defaultValue=<rgba 0.1683 0.2307 1 1>}
  n3 = TextureSample(UVs=Panner(Coordinate=TextureCoordinate() {vTiling=0.5}) {speedX=-0.05, speedY=-0.24}) {
   samplerSource=SSM_Wrap_WorldGroupSettings
   samplerType=SAMPLERTYPE_Normal
   texture=/GreaseClipDunmurry/Assets/Materials/Textures/T_MemoryTunnel_N1
  }
  n4 = DotProduct(A=@n3.RGB, B=Normalize(VectorInput=VectorParameter() {parameterName=LightDirection, defaultValue=<rgba 0 -1 0.4 0>}.RGB))
  n5 = Power(Base=Saturate(None=@n4)) {constExponent=2.5}
  n6 = MaterialFunctionCall(Float2=TextureCoordinate()) {materialFunction=&BreakOutFloat2Components}
  n7 = Multiply(A=Fresnel(Normal=NamedRerouteUsage()) {baseReflectFraction=0, exponent=2}) {constB=0.25}
  n8 = TextureSample!(
   UVs=Panner(Coordinate=AppendVector(A=@n6.R, B=@n7)) {speedX=0.04, speedY=0.1}
   Tex
   Bias=ScalarParameter() {parameterName=EmipBias}
   Apply View MipBias
  ) {mipValueMode=TMVM_MipBias, samplerSource=SSM_Wrap_WorldGroupSettings, texture=/Game/Effects/Fort_Effects/Textures/Tile/T_Caustics_Chromatic}
  n9 = Add(A=Multiply(A=LinearInterpolate(A=@n1.RGB, B=@n2.RGB, Alpha=@n5), B=Clamp(None=@n4) {minDefault=0.1}), B=Multiply(A=@n8.RGB) {constB=0.05})
  n10 = VectorParameter() {parameterName=DistortionTile, defaultValue=<rgba 3 2 0 0>}
  n11 = Multiply(A=TextureCoordinate() {uTiling=2, vTiling=2.5}, B=AppendVector(A=@n10.R, B=@n10.R))
  n12 = Multiply(A=ComponentMask(None=NamedRerouteUsage()) {g=true, r=true}) {constB=0.05}
  n13 = TextureSampleParameter2D(UVs=Panner(Coordinate=Add(A=@n11, B=@n12)) {speedX=0.05, speedY=-0.6}) {
   parameterName=DistortionTexture
   samplerSource=SSM_Wrap_WorldGroupSettings
   texture=/Game/Effects/Fort_Effects/Textures/Tile/T_ShieldCaustics_Tile_Chromatic
  }
  n14 = TextureCoordinate()
  n15 = Power(Base=TextureSample!(UVs, Tex, Apply View MipBias) {texture=/Game/Effects/Fort_Effects/Textures/Tile/T_Noise_02}.R) {constExponent=1}
  n16 = Constant()
  n17 = Multiply(A=Time()) {constB=-0.75}
  n18 = Add(
   A=OneMinus(None=MaterialFunctionCall(Float2=@n14) {materialFunction=&BreakOutFloat2Components}.G)
   B=Multiply(A=MaterialFunctionCall!(Width=@n15, Sharpness=@n16, Offset=@n17, Direction Switch, Compare, Input Coordinates) {materialFunction=/Engine/Functions/Engine_MaterialFunctions03/Procedurals/GeneratedBand}.Result) {constB=0.5}
  )
  n19 = Power(Base=@n18) {constExponent=0.75}
  n20 = TextureCoordinate()
  n21 = Panner(Coordinate=AppendVector(
    A=MaterialFunctionCall(Float2=@n20) {materialFunction=&BreakOutFloat2Components}.R
    B=Constant()
   )) {speedX=0.04, speedY=0.3}
  n22 = Multiply(
   A=LinearInterpolate(Alpha=TextureSample(UVs=@n21) {samplerSource=SSM_Wrap_WorldGroupSettings, texture=/Game/Global/Textures/Default/Noise/T_Tiling_Noise_01}.R) {constA=0.1}
   B=Multiply(A=Power(Base=@n18) {constExponent=1.5}) {constB=0.5}
  )
  n23 = Panner(Coordinate=TextureCoordinate() {uTiling=4, vTiling=2}) {speedY=-0.9}
  n24 = Add(A=Multiply(A=@n19, B=@n22), B=Multiply(A=@n22, B=TextureSample(UVs=@n23) {texture=/Game/Effects/FogVolume/noise_flamy_Tile}.B))
  n25 = Multiply(A=Multiply(A=Multiply(A=@n9, B=@n13.RGB)) {constB=2}, B=@n24)
  n26 = CurveAtlasRowParameter(CurveTime=@n19) {atlas=&/CA_GreaseClip_TimeTunnel, curve=&/C_GreaseClipTunnel}
  n27 = Multiply(A=@n26, B=@n24)
  n28 = NamedRerouteUsage()
  n29 = Add(A=Multiply(A=ComponentMask(None=@n28) {g=true, r=true}) {constB=0.04}, B=TextureCoordinate())
  n30 = Constant2Vector() {g=1.5, r=3}
  n31 = TextureSample(UVs=Panner(Coordinate=Multiply(A=@n29, B=@n30)) {speedX=0.05, speedY=-0.8}) {
   samplerSource=SSM_Wrap_WorldGroupSettings
   samplerType=SAMPLERTYPE_Alpha
   texture=/GreaseClipDunmurry/Assets/Materials/Textures/T_GreaseClip_Dunmurry_Sparks
  }
  n32 = NamedRerouteUsage()
  n33 = Add(A=Multiply(A=ComponentMask(None=@n32) {g=true, r=true}) {constB=0.03}, B=TextureCoordinate() {uTiling=3, vTiling=3})
  n34 = Constant2Vector() {g=2, r=4}
  n35 = TextureSample(UVs=Panner(Coordinate=Multiply(A=@n33, B=@n34)) {speedX=-0.02, speedY=-1.2}) {
   samplerSource=SSM_Wrap_WorldGroupSettings
   samplerType=SAMPLERTYPE_Alpha
   texture=&T_GreaseClip_Dunmurry_Sparks
  }
  n36 = Max(A=@n31.R, B=@n35.R)
  n37 = Multiply(
   A=LinearInterpolate(A=Constant3Vector() {constant=<rgba 0 0.2774 1 1>}, B=Constant3Vector() {constant=<rgba 0 0.8701 1 1>}, Alpha=@n36)
   B=@n36
  )
  n38 = ScalarParameter() {defaultValue=0.5, parameterName=SparksIntensity}
  n39 = Multiply(A=Add(A=Add(A=@n25, B=@n27), B=Multiply(A=@n37, B=@n38)), B=ScalarParameter() {defaultValue=3, parameterName=CoreEmissiveBrightness}) {constB=2}
  n40 = MaterialFunctionCall(Float2=TextureCoordinate()) {materialFunction=&BreakOutFloat2Components}
  n41 = Multiply(
   A=Power(Base=OneMinus(None=@n40.G)) {constExponent=1.5}
   B=VectorParameter() {parameterName=BGColor, defaultValue=<rgba 0.0002 0.0095 0.0448 1>}.RGB
  )
  n42 = Multiply(A=Add(A=Add(A=@n39, B=@n41), B=@n39), B=ScalarParameter() {defaultValue=3, parameterName=EmissiveBrightness})
  n43 = DynamicParameter() {paramNames [FadeIn,FadeOut,Param3,Param4]}
  n44 = Min(A=Divide(A=@n43.FadeIn) {constB=0.05}, B=Divide(A=@n43.FadeIn) {constB=0.2})
  LinearInterpolate(A=@n42, B=VectorParameter() {parameterName=FadeColor, defaultValue=<rgba 0.85 0.8788 1 1>}.RGB, Alpha=OneMinus(None=Saturate(None=@n44)))
  MaterialFunctionCall(Emissive=@n42) {materialFunction=&MF_FX_EmissiveAdjustments}
  NamedRerouteDeclaration(None=Transform(None=NamedRerouteDeclaration(None=@n3.RGB) {name=TangentSpaceNormal, nodeColor=<rgba 1 0 0.7969 1>})) {name=WorldSpaceNormal, nodeColor=<rgba 0 1 0.9297 1>}
  NamedRerouteDeclaration(None=@n13.RGB) {name=Distortion, nodeColor=<rgba 0 1 0.2734 1>}
 ]
}
