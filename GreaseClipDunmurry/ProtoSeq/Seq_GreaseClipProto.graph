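# Prototype GreaseClip Level Sequence: a cinematic of under six seconds showing an animated vehicle with driver and passenger mannequins and a backing screen.
# The mannequins are child actor components of a spawned manager, alongside spawned control and mockup actors, with a roof piece appearing partway through.
# One cine camera sweeps around the scene, layering short jolts from a second transform section and a camera shake over the main move.
# A fade-in, camera cut and three subsequences complete the shot.
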LevelSequence /GreaseClipDunmurry/ProtoSeq/Seq_GreaseClipProto
bindings [
 PassengerMannequin {
  id=#1
  parent_id=#2
  possessed_class=ChildActorComponent
  tracks [
   {type=MovieSceneSkeletalAnimationTrack, sections [{type=MovieSceneSkeletalAnimationSection, end_frame=150, start_frame=-8}]}
   {
    type=MovieScene3DTransformTrack
    sections [{
      type=MovieScene3DTransformSection
      channels [
       Location.X {type=MovieSceneScriptingDoubleChannel, default=-8}
       Location.Y {type=MovieSceneScriptingDoubleChannel, default=104}
       Location.Z {type=MovieSceneScriptingDoubleChannel, default=-4}
       Rotation.X {type=MovieSceneScriptingDoubleChannel, default=0}
       Rotation.Y {type=MovieSceneScriptingDoubleChannel, default=0}
       Rotation.Z {type=MovieSceneScriptingDoubleChannel, default=-90}
       Scale.X {type=MovieSceneScriptingDoubleChannel, default=1}
       Scale.Y {type=MovieSceneScriptingDoubleChannel, default=1}
       Scale.Z {type=MovieSceneScriptingDoubleChannel, default=1}
      ]
     }]
   }
  ]
 }
 GreaseClip_BackingScreenActor {
  id=#3
  possessed_class=GreaseClip_BackingScreenActor_C
  tracks [
   {type=MovieSceneSpawnTrack, sections [{type=MovieSceneSpawnSection, channels [None {type=MovieSceneScriptingBoolChannel, default=true}]}]}
   {
    type=MovieScene3DTransformTrack
    sections [{
      type=MovieScene3DTransformSection
      channels [
       Location.X {type=MovieSceneScriptingDoubleChannel, default=-2012}
       Location.Y {type=MovieSceneScriptingDoubleChannel, default=124}
       Location.Z {type=MovieSceneScriptingDoubleChannel, default=220}
       Rotation.X {type=MovieSceneScriptingDoubleChannel, default=0}
       Rotation.Y {type=MovieSceneScriptingDoubleChannel, default=0}
       Rotation.Z {type=MovieSceneScriptingDoubleChannel, default=-90}
       Scale.X {type=MovieSceneScriptingDoubleChannel, default=1.5}
       Scale.Y {type=MovieSceneScriptingDoubleChannel, default=1.5}
       Scale.Z {type=MovieSceneScriptingDoubleChannel, default=1}
      ]
     }]
   }
   {type=MovieScene3DAttachTrack, sections [{type=MovieScene3DAttachSection}]}
  ]
 }
 DriverMannequin {
  id=#4
  parent_id=#2
  possessed_class=ChildActorComponent
  tracks [
   {
    type=MovieSceneSkeletalAnimationTrack
    sections [{
      type=MovieSceneSkeletalAnimationSection
      end_frame=165
      start_frame=-8
      channels [Weight {type=MovieSceneScriptingFloatChannel, default=1, const [1,2]}]
     }]
   }
   {
    type=MovieScene3DTransformTrack
    sections [{
      type=MovieScene3DTransformSection
      channels [
       Location.X {type=MovieSceneScriptingDoubleChannel, default=-8}
       Location.Y {type=MovieSceneScriptingDoubleChannel, default=4}
       Location.Z {type=MovieSceneScriptingDoubleChannel, default=-12}
       Rotation.X {type=MovieSceneScriptingDoubleChannel, default=0}
       Rotation.Y {type=MovieSceneScriptingDoubleChannel, default=0}
       Rotation.Z {type=MovieSceneScriptingDoubleChannel, default=-90}
       Scale.X {type=MovieSceneScriptingDoubleChannel, default=1}
       Scale.Y {type=MovieSceneScriptingDoubleChannel, default=1}
       Scale.Z {type=MovieSceneScriptingDoubleChannel, default=1}
      ]
     }]
   }
  ]
 }
 Container_GreaseClip_Roof {
  id=#5
  possessed_class=FortStaticMeshActor
  tracks [
   {
    type=MovieSceneSpawnTrack
    sections [{type=MovieSceneSpawnSection, channels [None {type=MovieSceneScriptingBoolChannel, default=false, keys [[51,false],[70,true],[93,true]]}]}]
   }
   {
    type=MovieScene3DTransformTrack
    sections [{
      type=MovieScene3DTransformSection
      channels [
       Location.X {type=MovieSceneScriptingDoubleChannel, default=0}
       Location.Y {type=MovieSceneScriptingDoubleChannel, default=0}
       Location.Z {type=MovieSceneScriptingDoubleChannel, default=-40}
       Rotation.X {type=MovieSceneScriptingDoubleChannel, default=0}
       Rotation.Y {type=MovieSceneScriptingDoubleChannel, default=0}
       Rotation.Z {type=MovieSceneScriptingDoubleChannel, default=0}
       Scale.X {type=MovieSceneScriptingDoubleChannel, default=1}
       Scale.Y {type=MovieSceneScriptingDoubleChannel, default=1}
       Scale.Z {type=MovieSceneScriptingDoubleChannel, default=1}
      ]
     }]
   }
   {type=MovieScene3DAttachTrack, sections [{type=MovieScene3DAttachSection}]}
  ]
 }
 GreaseClipSequenceControlActor {
  id=#6
  possessed_class=GreaseClipSequenceControlActor_C
  tracks [
   {type=MovieSceneSpawnTrack, sections [{type=MovieSceneSpawnSection, channels [None {type=MovieSceneScriptingBoolChannel, default=true}]}]}
   {
    type=MovieScene3DTransformTrack
    sections [{
      type=MovieScene3DTransformSection
      channels [
       Location.X {type=MovieSceneScriptingDoubleChannel, default=0}
       Location.Y {type=MovieSceneScriptingDoubleChannel, default=0}
       Location.Z {type=MovieSceneScriptingDoubleChannel, default=50048.1869}
       Rotation.X {type=MovieSceneScriptingDoubleChannel, default=0}
       Rotation.Y {type=MovieSceneScriptingDoubleChannel, default=0}
       Rotation.Z {type=MovieSceneScriptingDoubleChannel, default=0}
       Scale.X {type=MovieSceneScriptingDoubleChannel, default=1}
       Scale.Y {type=MovieSceneScriptingDoubleChannel, default=1}
       Scale.Z {type=MovieSceneScriptingDoubleChannel, default=1}
      ]
     }]
   }
  ]
 }
 B_GreaseClip_MannequinManager {
  id=#2
  possessed_class=B_GreaseClip_MannequinManager_C
  tracks [
   {type=MovieSceneSpawnTrack, sections [{type=MovieSceneSpawnSection, channels [None {type=MovieSceneScriptingBoolChannel, default=true}]}]}
   {
    type=MovieScene3DTransformTrack
    sections [{
      type=MovieScene3DTransformSection
      channels [
       Location.X {type=MovieSceneScriptingDoubleChannel, default=-5}
       Location.Y {type=MovieSceneScriptingDoubleChannel, default=0}
       Location.Z {type=MovieSceneScriptingDoubleChannel, default=-30}
       Rotation.X {type=MovieSceneScriptingDoubleChannel, default=0}
       Rotation.Y {type=MovieSceneScriptingDoubleChannel, default=0}
       Rotation.Z {type=MovieSceneScriptingDoubleChannel, default=0}
       Scale.X {type=MovieSceneScriptingDoubleChannel, default=1}
       Scale.Y {type=MovieSceneScriptingDoubleChannel, default=1}
       Scale.Z {type=MovieSceneScriptingDoubleChannel, default=1}
      ]
     }]
   }
   {type=MovieScene3DAttachTrack, sections [{type=MovieScene3DAttachSection}]}
  ]
 }
 GreaseClipCineCamera {
  id=#7
  possessed_class=GreaseClipCineCamera_C
  tracks [
   {type=MovieSceneSpawnTrack, sections [{type=MovieSceneSpawnSection, channels [None {type=MovieSceneScriptingBoolChannel, default=true}]}]}
   {
    type=MovieScene3DTransformTrack
    sections [
     {
      type=MovieScene3DTransformSection
      channels [
       Location.X {
        type=MovieSceneScriptingDoubleChannel
        default=120
        keys [[-6,-280.345],[1,-161.5208],[9,37.8895],[28,428.391],[44,139.7766],[60,-185.0518],[74,-229.915],[120,-255.6533],[164,-258.8875],[172,-261.6391]]
       }
       Location.Y {
        type=MovieSceneScriptingDoubleChannel
        default=23
        keys [[-5,379.6272],[5,378.6716],[40,436.8032],[58,350.2405],[79,256.8686],[120,268.4139],[132,256.2399],[152,35.5222],[164,23],[172,22.4045]]
       }
       Location.Z {
        type=MovieSceneScriptingDoubleChannel
        default=90
        keys [[-5,188.2598],[1,169.5257],[6,120.2615],[12,56.5338],[24,31.1532],[40,45.1699],[74,51.265],[119,56.8773],[141,59.0526],[156,87.7921],[164,90],[172,90.3514]]
       }
       Rotation.X {type=MovieSceneScriptingDoubleChannel, default=0, keys [[-5,-10],[4,-9.2573],[60,-6.6271],[83.9825,-6.5866],[154,-0.0299],[164,0],[172,0]]}
       Rotation.Y {
        type=MovieSceneScriptingDoubleChannel
        default=0
        keys [[-5.0013,-3],[2,2.3221],[5,4.5],[7,3.3893],[9,-1.6498],[12,-7.4832],[24,-5.0032],[61,1.3266],[75,2.2794],[87,2.2327],[119,0.942],[164,0],[172,0]]
       }
       Rotation.Z {
        type=MovieSceneScriptingDoubleChannel
        default=0
        thinned_from=13
        keys [[-5,-150.7549],[2,-135.0567],[3,-131.4656],[5,-117.0504],[8,-106.9573],[25,-86.8861],[60,-37],[118,-36.5447],[132,-25.5601],[152,4.9737],[164,5.2337],[172,5.2337]]
       }
       Scale.X {type=MovieSceneScriptingDoubleChannel, default=1}
       Scale.Y {type=MovieSceneScriptingDoubleChannel, default=1}
       Scale.Z {type=MovieSceneScriptingDoubleChannel, default=1}
      ]
     }
     {
      type=MovieScene3DTransformSection
      channels [
       Location.X {type=MovieSceneScriptingDoubleChannel, default=0, const [0,3]}
       Location.Y {type=MovieSceneScriptingDoubleChannel, default=0, const [0,3]}
       Location.Z {
        type=MovieSceneScriptingDoubleChannel
        default=0
        keys [[-5,0],[17,0],[20,-1.4632],[24,1.7248],[28,-0.7494],[32,0.0015],[73,0],[76,-1.4632],[79,1.7248],[82,-0.7494],[85,0.0015],[164,0],[172,0]]
       }
       Rotation.X {
        type=MovieSceneScriptingDoubleChannel
        default=0
        keys [[-5,0],[18,0.0001],[21,-0.8378],[25,0.3574],[30,-0.2287],[33,-0.004],[164,0],[172,0]]
       }
       Rotation.Y {
        type=MovieSceneScriptingDoubleChannel
        default=0
        thinned_from=13
        keys [[-5,0],[18,0.0023],[21,1.7623],[25,-1.5007],[29,1.0316],[33,0.0015],[74,0.0015],[77,1.7615],[80,-1.5015],[83,1.0308],[86,0.0007],[172,0]]
       }
       Rotation.Z {
        type=MovieSceneScriptingDoubleChannel
        default=0
        thinned_from=8
        keys [[-5,0],[73,0.0015],[76,-2.6566],[79,1.0479],[82,-0.3448],[85,0.0007],[172,0]]
       }
       Scale.X {type=MovieSceneScriptingDoubleChannel, default=1, const [1,1]}
       Scale.Y {type=MovieSceneScriptingDoubleChannel, default=1, const [1,1]}
       Scale.Z {type=MovieSceneScriptingDoubleChannel, default=1, const [1,1]}
      ]
     }
    ]
   }
   {type=MovieSceneCameraShakeTrack, sections [{type=MovieSceneCameraShakeSection, end_frame=173, start_frame=-24}]}
   {type=MovieScene3DAttachTrack, sections [{type=MovieScene3DAttachSection}]}
  ]
 }
 SkeletalMeshComponent0 {id=#8, parent_id=#9, possessed_class=SkeletalMeshComponent, tracks []}
 CameraComponent {
  id=#10
  parent_id=#7
  possessed_class=CineCameraComponent
  tracks [
   {
    type=MovieSceneFloatTrack
    display_name=Current Focal Length
    sections [{type=MovieSceneFloatSection, channels [None {type=MovieSceneScriptingFloatChannel, default=15.7631}]}]
   }
   {
    type=MovieSceneFloatTrack
    display_name="Manual Focus Distance (Focus Settings)"
    sections [{type=MovieSceneFloatSection, channels [None {type=MovieSceneScriptingFloatChannel, default=100000}]}]
   }
   {
    type=MovieSceneFloatTrack
    display_name=Current Aperture
    sections [{type=MovieSceneFloatSection, channels [None {type=MovieSceneScriptingFloatChannel, default=2.9716}]}]
   }
  ]
 }
 Vehicle+Animation {
  id=#9
  possessed_class=FortSkeletalMeshActor
  tracks [
   {type=MovieSceneSpawnTrack, sections [{type=MovieSceneSpawnSection, channels [None {type=MovieSceneScriptingBoolChannel, default=true}]}]}
   {
    type=MovieScene3DTransformTrack
    sections [{
      type=MovieScene3DTransformSection
      channels [
       Location.X {type=MovieSceneScriptingDoubleChannel, default=0}
       Location.Y {type=MovieSceneScriptingDoubleChannel, default=0}
       Location.Z {type=MovieSceneScriptingDoubleChannel, default=-64}
       Rotation.X {type=MovieSceneScriptingDoubleChannel, default=0}
       Rotation.Y {type=MovieSceneScriptingDoubleChannel, default=0}
       Rotation.Z {type=MovieSceneScriptingDoubleChannel, default=-90}
       Scale.X {type=MovieSceneScriptingDoubleChannel, default=1}
       Scale.Y {type=MovieSceneScriptingDoubleChannel, default=1}
       Scale.Z {type=MovieSceneScriptingDoubleChannel, default=1}
      ]
     }]
   }
   {type=MovieSceneSkeletalAnimationTrack, sections [{type=MovieSceneSkeletalAnimationSection, end_frame=161, start_frame=-5}]}
   {type=MovieScene3DAttachTrack, sections [{type=MovieScene3DAttachSection}]}
  ]
 }
 BP_GreaseClipProtoSeqMockup {
  id=#11
  possessed_class=BP_GreaseClipProtoSeqMockup_C
  tracks [
   {type=MovieSceneSpawnTrack, sections [{type=MovieSceneSpawnSection, channels [None {type=MovieSceneScriptingBoolChannel, default=true}]}]}
   {
    type=MovieScene3DTransformTrack
    sections [{
      type=MovieScene3DTransformSection
      channels [
       Location.X {type=MovieSceneScriptingDoubleChannel, default=0}
       Location.Y {type=MovieSceneScriptingDoubleChannel, default=0}
       Location.Z {type=MovieSceneScriptingDoubleChannel, default=0}
       Rotation.X {type=MovieSceneScriptingDoubleChannel, default=0}
       Rotation.Y {type=MovieSceneScriptingDoubleChannel, default=0}
       Rotation.Z {type=MovieSceneScriptingDoubleChannel, default=0}
       Scale.X {type=MovieSceneScriptingDoubleChannel, default=1}
       Scale.Y {type=MovieSceneScriptingDoubleChannel, default=1}
       Scale.Z {type=MovieSceneScriptingDoubleChannel, default=1}
      ]
     }]
   }
   {type=MovieScene3DAttachTrack, sections [{type=MovieScene3DAttachSection}]}
  ]
 }
]
display_rate {denominator=1, numerator=30}
tick_resolution {denominator=1, numerator=24000}
tracks [
 {
  type=MovieSceneFadeTrack
  sections [{type=MovieSceneFadeSection, channels [None {type=MovieSceneScriptingFloatChannel, default=1, keys [[-6,1],[15,0]]}]}]
 }
 {type=MovieSceneSubTrack, display_name=Subsequences, sections [{type=MovieSceneSubSection, end_frame=165, start_frame=-5}]}
 {type=MovieSceneSubTrack, display_name=Subsequences, sections [{type=MovieSceneSubSection, end_frame=165, start_frame=-5}]}
 {type=MovieSceneSubTrack, display_name=Subsequences, sections [{type=MovieSceneSubSection, end_frame=165, start_frame=-9}]}
 {type=MovieSceneCameraCutTrack, sections [{type=MovieSceneCameraCutSection, end_frame=165, start_frame=-5}]}
]
